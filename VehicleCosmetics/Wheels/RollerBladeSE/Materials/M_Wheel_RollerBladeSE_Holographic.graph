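# Holographic Default Lit material for the RollerBlade SE vehicle wheel, with separate looks for body and blinking rollers.
# The body layers custom and accent colors with Fresnel, scrolling wave glow, parallax glitter sparkles and pulsing UV1 rings.
# A RollerMaskForLOD static switch takes roller masks from a texture instead of vertex color and also masks body emissive.
# The combined attributes pass through the shared vehicle damage and de-rez transition functions.
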Material /VehicleCosmetics/Wheels/RollerBladeSE/Materials/M_Wheel_RollerBladeSE_Holographic
graph {
 properties {
  bAutomaticallySetUsageInEditor=false
  bUseMaterialAttributes=true
  bUsedWithMeshDeformer=true
  bUsedWithSkeletalMesh=true
  usedShadingModels=Default Lit
 }
 nodes [
  n1 = NamedRerouteUsage()
  n2 = SetMaterialAttributes!(
   MaterialAttributes
   Base Color=LinearInterpolate(A=NamedRerouteUsage(), B=NamedRerouteUsage(), Alpha=@n1)
   Emissive Color=LinearInterpolate(A=NamedRerouteUsage(), B=NamedRerouteUsage(), Alpha=@n1)
   Normal=NamedRerouteUsage()
   Metallic=ScalarParameter() {group=00_Base, parameterName=Metalic}
   Roughness=ScalarParameter() {defaultValue=0.5, group=00_Base, parameterName=Roughness}
  ) {attributeSetTypes [
    #1
    #2
    #3
    #4
    #5
   ]}
  MaterialFunctionCall(In=MaterialFunctionCall(Material=@n2) {materialFunction=&MF_BaseVehicle_Damage}.Result) {materialFunction=&MF_VehicleDeRez_Transition}
  n3 = MaterialFunctionCall!(Normal, Exponent=Constant() {r=3}) {materialFunction=/VehicleCosmetics/Wheels/Cola_Holographic/Materials/MF_Fresnel_Classic}
  n4 = TextureSampleParameter2D() {
   group=00_Base
   parameterName=Normal
   samplerType=SAMPLERTYPE_Normal
   sortPriority=3
   texture=/VehicleCosmetics/Wheels/RollerBladeSE/Textures/T_RollerBlades_SE_N_LOD
  }
  n5 = NamedRerouteUsage()
  n6 = NamedRerouteDeclaration(None=LinearInterpolate(
    A=@n4.RGB
    B=TextureSample() {samplerType=SAMPLERTYPE_Normal, texture=/VehicleCosmetics/Wheels/_SharedTextures/FX/normal_flat}.RGB
    Alpha=@n5
   )) {name=Normal, nodeColor=<rgba 1 0.5391 0 1>}
  n7 = Constant() {r=3}
  n8 = Max(A=Abs(None=@n4.R), B=Abs(None=LinearInterpolate(A=@n4.R, Alpha=@n5) {constB=0}))
  n9 = Clamp(None=ConstantBiasScale(None=Clamp(None=@n8)) {bias=-0.1, scale=5}) {maxDefault=5}
  n10 = Multiply(A=Time(), B=ScalarParameter() {defaultValue=1, group=00_Base, parameterName=Wave Time Speed, sortPriority=12}) {constB=0.025}
  n11 = OneMinus(None=Sine(None=@n10) {period=2})
  n12 = ScalarParameter() {defaultValue=0.5, group=00_Base, parameterName=Wave Glow Intensity, sortPriority=13}
  n13 = Add(A=Multiply(A=Clamp(None=@n11), B=@n12)) {constB=0}
  n14 = Multiply(A=@n9, B=@n13)
  n15 = Max(
   A=Multiply(
    A=MaterialFunctionCall(Normal=@n6, Exponent=@n7) {materialFunction=&MF_Fresnel_Classic}.Result
    B=@n8
   )
   B=LinearInterpolate(A=Multiply(A=@n14, B=@n13), B=@n14, Alpha=Add())
  )
  n16 = Multiply(A=@n15, B=ScalarParameter() {defaultValue=1, group=00_Base, parameterName=Emissive Intensity, sortPriority=10})
  n17 = Multiply(A=LinearInterpolate(A=@n9, B=@n15, Alpha=NamedRerouteUsage()), B=NamedRerouteUsage()) {desc=BASE COLOR}
  n18 = TextureCoordinate()
  n19 = AppendVector(A=ComponentMask(None=@n18) {g=true}, B=ComponentMask(None=@n18) {r=true})
  n20 = OneMinus(None=TextureSample(UVs=Panner(Coordinate=@n19, Time=@n10) {speedX=-0.5}) {texture=/VehicleCosmetics/Wheels/_SharedTextures/FX/SharpGradient_01_Heightmap}.R)
  n21 = Clamp(None=ConstantBiasScale(None=Clamp(None=@n20)) {bias=-0.9, scale=0.25})
  n22 = Multiply(
   A=Multiply(A=ConstantBiasScale(None=@n16) {bias=0.02, scale=1}, B=NamedRerouteDeclaration(None=@n17) {name=Base Color, nodeColor=<rgba 1 0.8906 0 1>})
   B=Multiply(A=Multiply(A=@n21) {constA=40, constB=40}, B=@n12)
  )
  n23 = NamedRerouteUsage()
  n24 = Add(A=Multiply(A=Multiply(A=@n3.Result) {constB=2}, B=NamedRerouteUsage()) {constB=2}, B=Add(A=Add(A=@n22, B=@n23), B=NamedRerouteUsage()))
  n25 = NamedRerouteUsage()
  NamedRerouteDeclaration(None=StaticSwitchParameter(True=LinearInterpolate(A=@n24, Alpha=@n25) {constB=0}, False=@n24) {group=00_Base, parameterName=RollerMaskForLOD, sortPriority=3}) {name=Emissive, nodeColor=<rgba 1 0.8906 0 1>}
  n26 = TextureSampleParameter2D() {
   group=00_Base
   parameterName=Masks
   samplerType=SAMPLERTYPE_LinearColor
   sortPriority=2
   texture=/VehicleCosmetics/Wheels/RollerBladeSE/Textures/T_RollerBlades_SE_M_LOD
  }
  NamedRerouteDeclaration(None=@n26.R) {name=Masks_R, nodeColor=<rgba 1 0 0 1>}
  NamedRerouteDeclaration(None=@n26.G) {name=Masks_G, nodeColor=<rgba 0 1 0 1>}
  NamedRerouteDeclaration(None=@n26.B) {name=Masks_B, nodeColor=<rgba 0 0 1 1>}
  n27 = Power(Base=MaterialFunctionCall(Normal=NamedRerouteUsage(), Exponent=Constant() {r=3}) {materialFunction=&MF_Fresnel_Classic}.Result) {constExponent=0.01}
  n28 = OneMinus(None=Power(Base=Clamp(None=@n27)) {constExponent=3})
  n29 = NamedRerouteUsage()
  NamedRerouteDeclaration(None=Add(A=Multiply(A=@n28, B=@n29), B=NamedRerouteUsage())) {name=Color+Fresnel, nodeColor=<rgba 0.1016 1 0 1>}
  n30 = CameraVectorWS()
  n31 = TextureSample(UVs=ComponentMask(None=Transform(None=@n30) {transformSourceType=TRANSFORMSOURCE_View}) {g=true, r=true}) {texture=/VehicleCosmetics/Wheels/_SharedTextures/FX/Glittermap}
  n32 = OneMinus(None=Clamp(None=Multiply(A=@n31.RGB) {constB=0.5}))
  n33 = Multiply(A=Multiply(A=@n32, B=@n32)) {constB=0.25}
  n34 = Normalize(VectorInput=CameraVectorWS())
  n35 = Multiply(A=Time()) {constB=0.025}
  n36 = Constant() {r=-0.5}
  n37 = BumpOffset(Coordinate=Rotator(Coordinate=TextureCoordinate() {uTiling=0.5, vTiling=0.2}, Time=@n35) {speed=0.2}, Height=@n36)
  n38 = BumpOffset(Coordinate=Panner(Coordinate=TextureCoordinate() {uTiling=0.3, vTiling=0.2}, Time=@n35) {speedX=-0.035, speedY=-0.045}, Height=@n36)
  n39 = Subtract(
   A=Multiply(
    A=TextureSample(UVs=@n37) {samplerType=SAMPLERTYPE_LinearColor, texture=/VehicleCosmetics/Wheels/_SharedTextures/FX/paintfinish_glitter}.RGB
    B=TextureSample(UVs=@n38) {samplerType=SAMPLERTYPE_LinearColor, texture=&paintfinish_glitter}.RGB
   )
   B=ScalarParameter() {defaultValue=0.52, group=00_Base, parameterName=Glitter Density, sortPriority=11}
  )
  n40 = Clamp(None=DotProduct(A=OneMinus(None=@n34), B=Normalize(VectorInput=@n39))) {minDefault=-0.01}
  NamedRerouteDeclaration(None=LinearInterpolate(B=LinearInterpolate(B=@n33, Alpha=@n40), Alpha=NamedRerouteUsage())) {name=Sparkles, nodeColor=<rgba 1 0.7031 0 1>}
  n41 = VectorParameter() {group=00_Base, parameterName=CustomColor, sortPriority=1, defaultValue=<rgba 0 0.0505 0.5525 1>}
  n42 = Desaturation(
   None=LinearInterpolate(
    A=@n41.RGB
    B=VectorParameter() {group=00_Base, parameterName=AccentColor, sortPriority=0, defaultValue=<rgba 0 0.9613 0.5136 1>}.RGB
    Alpha=NamedRerouteUsage()
   )
   Fraction=Constant()
  ) {luminanceFactors=<rgba 0.3 0.59 0.11 0>}
  NamedRerouteDeclaration(None=LinearInterpolate(A=NamedRerouteDeclaration(None=@n42) {name=Color Desat, nodeColor=<rgba 0.2188 0 1 1>}, B=@n41.RGB, Alpha=NamedRerouteUsage())) {name=Color, nodeColor=<rgba 0.1016 1 0 1>}
  n43 = TextureCoordinate() {coordinateIndex=1}
  n44 = Distance(
   A=Multiply(A=Subtract(A=@n43) {constB=0.5}, B=ScalarParameter() {defaultValue=0.25, group=00_Base, parameterName=Rings Bands Amount, sortPriority=14})
   B=Constant2Vector()
  )
  n45 = Multiply(A=Time(), B=ScalarParameter() {defaultValue=-0.05, group=00_Base, parameterName=Rings Pulse Speed, sortPriority=15}) {constB=0.025}
  n46 = Add(A=Cosine(None=Add(A=@n44, B=@n45)) {period=0.075})
  n47 = Clamp(None=Power(Base=Divide(A=@n46)) {constExponent=60})
  n48 = Multiply(A=NamedRerouteUsage()) {constB=0.1}
  n49 = NamedRerouteUsage()
  NamedRerouteDeclaration(None=Multiply(A=Multiply(A=@n47) {constB=5}, B=Add(A=@n48, B=@n49))) {name=EmissiveRings, nodeColor=<rgba 0 0.9062 1 1>}
  n50 = TextureSampleParameter2D() {
   group=00_Base
   parameterName=RollerMask
   sortPriority=3
   texture=/VehicleCosmetics/Wheels/RollerBladeSE/Textures/T_RollerBlades_SE_M_Wheel_LOD
  }
  n51 = VertexColor()
  NamedRerouteDeclaration(None=StaticSwitchParameter(True=@n50.R, False=@n51.R) {group=00_Base, parameterName=RollerMaskForLOD, sortPriority=3}) {name=VC_R, nodeColor=<rgba 1 0 0 1>}
  n52 = VectorParameter() {group=00_Base, parameterName=Roller Color, sortPriority=0, defaultValue=<rgba 0.399 0.0161 0.399 1>}
  NamedRerouteDeclaration(None=@n52.RGB) {name=Custom Color, nodeColor=<rgba 0.1484 1 0 1>}
  NamedRerouteDeclaration(None=StaticSwitchParameter(True=@n50.A, False=@n51.A) {group=00_Base, parameterName=RollerMaskForLOD, sortPriority=3}) {name=VC_A, nodeColor=<rgba 1 1 1 1>}
  n53 = TextureSampleParameter2D() {group=00_Base, parameterName=Base Color, sortPriority=2, texture=/VehicleCosmetics/Wheels/RollerBladeSE/Textures/T_RollerBlades_SE_BC_LOD}
  n54 = NamedRerouteUsage()
  n55 = Constant() {r=3}
  n56 = Clamp(None=Power(Base=MaterialFunctionCall(Normal=@n54, Exponent=@n55) {materialFunction=&MF_Fresnel_Classic}.Result) {constExponent=0.01})
  n57 = OneMinus(None=Power(Base=@n56) {constExponent=20})
  n58 = Multiply(A=@n57, B=@n57)
  n59 = Multiply(A=Multiply(A=Multiply(A=@n58, B=@n58), B=NamedRerouteUsage()), B=NamedRerouteUsage())
  NamedRerouteDeclaration(None=Multiply(A=Add(A=@n53.R, B=@n53.R), B=@n59)) {name=Roller_BC, nodeColor=<rgba 0.5938 1 0 1>}
  n60 = OneMinus(None=@n56)
  n61 = Time()
  n62 = ScalarParameter() {defaultValue=0.5, group=00_Base, parameterName=RollerTimeSpeed, sortPriority=5}
  n63 = Frac(None=Add(A=Multiply(A=@n61, B=@n62), B=NamedRerouteUsage()))
  n64 = Multiply(A=NamedRerouteUsage(), B=ScalarParameter() {defaultValue=5, group=00_Base, parameterName=RollerBlinkBrightness, sortPriority=6})
  n65 = Constant()
  NamedRerouteDeclaration(None=Add(A=Add(A=@n60, B=@n59), B=If(A=@n63, "A > B"=@n64, "A == B"=@n65, "A < B"=@n65) {constB=0.9})) {name=Roller_Emissive, nodeColor=<rgba 0.5938 1 0 1>}
  NamedRerouteDeclaration(None=@n26.A) {name=Masks_A, nodeColor=<rgba 1 1 1 1>}
  NamedRerouteDeclaration(None=@n50.G) {name=Roller Blakc Mask LOD, nodeColor=<rgba 0 1 0 1>}
 ]
}
